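# Stylized flame master material for Niagara sprites, ribbons and meshes: unlit, translucent and two-sided.
# Panned noise distorts the main flame texture, which combines an optional mask and UV gradients to drive border-to-core color and opacity.
# Dynamic parameters control distorted noise erosion and optional custom time, while edge, depth and view-angle fades shape alpha.
# World position offset adds panning sine wobble weighted along UVs, plus optional SplineThicken.
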Material /BRCosmetics/Effects/Niagara/MasterMaterials/Fire/M_Flame_Stylized
graph {
 properties {
  bAutomaticallySetUsageInEditor=false
  bUsedWithNiagaraMeshParticles=true
  bUsedWithNiagaraRibbons=true
  bUsedWithNiagaraSprites=true
  blendMode=BLEND_Translucent
  shadingModel=MSM_Unlit
  twoSided=true
  usedShadingModels=Unlit
 }
 nodes [
  n1 = ScalarParameter() {defaultValue=5, parameterName=WidthBase}
  n2 = ScalarParameter() {defaultValue=1.25, parameterName=WidthTip}
  n3 = TextureCoordinate()
  n4 = Power(Base=OneMinus(None=ComponentMask(None=@n3) {g=true}))
  n5 = Panner(Coordinate=@n3) {speedY=0.7}
  n6 = Sine(None=Multiply(A=Panner(Coordinate=@n3) {speedY=0.4}) {constB=2})
  n7 = ComponentMask(None=Multiply(A=Sine(None=@n5), B=Multiply(A=@n6) {constB=0.4})) {g=true}
  n8 = Multiply(
   A=Subtract(A=Multiply(A=@n7) {constB=2})
   B=Transform(None=Constant3Vector() {constant=<rgba 0 1 1.1 0>}) {transformSourceType=TRANSFORMSOURCE_Instance}
  )
  Add(
   A=LinearInterpolate(
    B=MaterialFunctionCall(WidthBase=@n1, WidthTip=@n2) {materialFunction=&SplineThicken}.WorldPositionOffset
    Alpha=ScalarParameter() {parameterName=UseSplineThicken}
   )
   B=Multiply(A=Multiply(A=@n4, B=@n8), B=ScalarParameter() {defaultValue=1, parameterName=WPOIntensity})
  )
  n9 = ParticleRandom()
  n10 = Constant()
  n11 = Add(A=TextureCoordinate(), B=LinearInterpolate(B=AppendVector(A=@n9, B=@n10), Alpha=ScalarParameter() {parameterName=AddParticleRandom}))
  n12 = VectorParameter() {
   group=Distortion
   parameterName=DistortionUVControl
   sortPriority=2
   channelNames {a=SpeedY, b=SpeedX, g=ScaleY, r=ScaleX}
   defaultValue=<rgba 0.6 0.6 -1 0.25>
  }
  n13 = ComponentMask(None=@n12) {g=true, r=true}
  n14 = LinearInterpolate(
   A=Time()
   B=DynamicParameter() {defaultValue=<rgba 0 0.1 0 1>, paramNames [Erosion, ErosionSmoothness, DistortionErosionAmount, Time]}.Time
   Alpha=ScalarParameter() {parameterName=UseCustomTime}
  )
  n15 = TextureSampleParameter2D(UVs=Panner(Coordinate=Multiply(A=@n11, B=@n13) {constB=0.6}, Time=@n14, Speed=AppendVector(A=@n12.SpeedX, B=@n12.SpeedX)) {speedX=1, speedY=0.25}) {
   group=Distortion
   parameterName=DistortionTexture
   samplerType=SAMPLERTYPE_Grayscale
   sortPriority=1
   texture=/Game/Effects/Fort_Effects/Textures/Fire/T_WispyFireNoise_Soft_M
  }
  n16 = AppendVector(
   A=Constant()
   B=Multiply(
    A=ChannelMaskParameter(None=@n15.RGB) {group=Distortion, parameterName=DistortionChannelMaskforMain, sortPriority=5}.R
    B=ScalarParameter() {defaultValue=0.2, group=Distortion, parameterName=DistortionAmount, sortPriority=3}
   ) {constB=0.2}
  )
  n17 = Add(A=@n11, B=@n16)
  n18 = VectorParameter() {
   group=MainTexture
   parameterName=FlameTextureUVControl
   sortPriority=2
   channelNames {a=SpeedY, b=SpeedX, g=ScaleY, r=ScaleX}
   defaultValue=<rgba 0.5 1 -0.6 0>
  }
  n19 = ComponentMask(None=@n18) {g=true, r=true}
  n20 = TextureSampleParameter2D(UVs=Panner(Coordinate=Multiply(A=@n17, B=@n19), Time=@n14, Speed=AppendVector(A=@n18.SpeedX, B=@n18.SpeedX)) {speedX=0.6}) {
   group=MainTexture
   parameterName=MainTexture
   sortPriority=1
   texture=/BRCosmetics/Athena/Effects/Particles/Skydiving/NS_Contrail_Chainmail/T_Chainmail_WispyMaskFat
  }
  n21 = ChannelMaskParameter(None=@n20.RGB) {group=MainTexture, parameterName=MainChannelMask, sortPriority=5}
  n22 = VectorParameter() {group=MaskTexture, parameterName=MaskTextureUVControl, channelNames {a=SpeedY, b=SpeedX, g=ScaleY, r=ScaleX}, defaultValue=<rgba 1 1 0 0>}
  n23 = Multiply(A=TextureCoordinate(), B=ComponentMask(None=@n22.RGB) {g=true, r=true})
  n24 = AppendVector(A=@n22.SpeedX, B=@n22.SpeedX)
  n25 = ChannelMaskParameter(None=TextureSampleParameter2D(UVs=Panner(Coordinate=@n23, Speed=@n24)) {group=MaskTexture, parameterName=MaskTexture, texture=/Game/Effects/Fort_Effects/Textures/Smoke/T_Clouds_02_M}.RGB) {group=MaskTexture, parameterName=MaskChannelMask, sortPriority=5}
  n26 = TextureCoordinate()
  n27 = Power(Base=OneMinus(None=ComponentMask(None=@n26) {r=true})) {constExponent=1}
  n28 = Multiply(A=ComponentMask(None=@n26) {g=true}) {constB=0.5}
  n29 = Multiply(A=Saturate(None=Multiply(A=@n27) {constB=2}), B=Power(Base=Sine(None=@n28)))
  n30 = TextureCoordinate()
  n31 = Add(A=@n30, B=@n16)
  n32 = Power(Base=OneMinus(None=ComponentMask(None=@n31) {r=true})) {constExponent=3}
  n33 = Add(
   A=StaticSwitch(True=Multiply(A=@n21.R, B=@n25.R), False=@n21.R, Value=StaticBoolParameter() {group=MaskTexture, parameterName=Use Mask?})
   B=Multiply(A=Multiply(A=@n29, B=@n32), B=ScalarParameter() {defaultValue=1, parameterName=Param_1})
  )
  n34 = ScalarParameter() {parameterName=TextureSubtract}
  n35 = Subtract(
   A=Saturate(None=Subtract(A=@n33, B=@n34) {constB=0.5})
   B=Power(Base=ComponentMask(None=@n30) {r=true}, Exp=ScalarParameter() {defaultValue=3}) {constExponent=3}
  )
  n36 = LinearInterpolate(
   A=VectorParameter() {group=Color, parameterName=BorderColor, sortPriority=1, defaultValue=<rgba 0.4271 0.0515 0 1>}.RGB
   B=VectorParameter() {group=Color, parameterName=CoreColor, sortPriority=2, defaultValue=<rgba 2 0.3435 0 1>}.RGB
   Alpha=Power(Base=Saturate(None=@n35)) {constExponent=3}
  )
  n37 = ParticleColor()
  MaterialFunctionCall(Emissive=Desaturation(None=Multiply(A=@n36, B=@n37.RGB), Fraction=ScalarParameter() {defaultValue=0.15, group=Color, parameterName=Desaturation, sortPriority=3})) {materialFunction=&MF_FX_EmissiveAdjustments}
  n38 = Multiply(A=@n35, B=ScalarParameter() {defaultValue=3, group=Color, parameterName=AlphaBoost, sortPriority=4}) {constB=3}
  n39 = TextureCoordinate()
  n40 = Multiply(
   A=Multiply(A=@n37.A, B=Saturate(None=@n38))
   B=SmoothStep(Max=ScalarParameter() {defaultValue=0.2, parameterName=EdgeMaskMax}, Value=ComponentMask(None=@n39) {r=true})
  )
  n41 = LinearInterpolate(
   A=@n40
   B=DepthFade(Opacity=@n40, FadeDistance=ScalarParameter() {parameterName=DepthFade})
   Alpha=ScalarParameter() {parameterName=UseDepthFade}
  )
  n42 = VertexNormalWS()
  n43 = CameraVectorWS()
  n44 = Power(Base=Abs(None=DotProduct(A=@n42, B=@n43)), Exp=ScalarParameter() {defaultValue=1, parameterName=AngleFadeIntensity}) {constExponent=1}
  Multiply(A=LinearInterpolate(A=@n41, B=Multiply(A=@n41, B=@n44), Alpha=ScalarParameter() {parameterName=UseAngleFade}), B=NamedRerouteUsage())
  n45 = DynamicParameter() {defaultValue=<rgba 0 0.1 0 1>, paramNames [Erosion, ErosionSmoothness, DistortionErosionAmount, Time]}
  n46 = VectorParameter() {
   group=Erosion
   parameterName=ErosionUVControl
   sortPriority=3
   channelNames {a=SpeedY, b=SpeedX, g=ScaleY, r=ScaleX}
   defaultValue=<rgba 1 1 0 0>
  }
  n47 = Multiply(A=TextureCoordinate(), B=ComponentMask(None=@n46) {g=true, r=true})
  n48 = AppendVector(A=@n46.SpeedX, B=@n46.SpeedX)
  n49 = TextureSampleParameter2D(UVs=Add(A=NamedRerouteUsage(), B=Panner(Coordinate=@n47, Time=@n14, Speed=@n48))) {
   group=Erosion
   parameterName=ErosionText
   samplerType=SAMPLERTYPE_LinearGrayscale
   sortPriority=0
   texture=/Game/Effects/Fort_Effects/Emotes/UltraLight/Texture/T_Tiling_Noise_01_GreyScale
  }
  NamedRerouteDeclaration(None=SmoothStep(Min=Subtract(A=@n45.Erosion, B=@n45.Erosion), Max=Add(A=@n45.Erosion, B=@n45.Erosion), Value=@n49.R)) {name=Erosion, nodeColor=<rgba 0.4297 1 0 1>}
  n50 = ChannelMaskParameter(None=@n15.RGB) {group=Distortion, maskChannel=Green, parameterName=DistortionChannelMaskforErosion, sortPriority=5, defaultValue=<rgba 0 1 0 0>}
  NamedRerouteDeclaration(None=Multiply(A=@n50.R, B=@n45.DistortionErosionAmount)) {name=DistForErosion, nodeColor=<rgba 1 0 0.3984 1>}
 ]
}
